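# One-shot mono MetaSound source that plays the out-of-ammo sound for the Muster pump shotgun (UnVault variant).
# On play, a shuffle array picks one of three wave variations and triggers the wave player, which fires OnFinished when done.
# Each play gets a random pitch shift within a base plus or minus spread range.
# It uses weapon foley attenuation and sound class; the MasterVolume input is not wired into the graph.
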MetaSoundSource /MusterCoreWeapons/Items/Weapons/MusterPumpShotgun/Audio/UnVault/MSS_Shotgun_MusterPump_Athena_UnVault_OutOfAmmo
inputs [
 MasterVolume {type=Float}
 Random.PitchShift.Base {type=Float}
 Random.PitchShift.Spread {type=Float}
 Sounds {type=WaveAsset:Array}
 UE.Source.OnPlay {type=Trigger}
]
interfaces [UE.OutputFormat.Mono:1.0, UE.Source.OneShot:1.0, UE.Source:1.1]
outputs [
 "UE.OutputFormat.Mono.Audio:0" {type=Audio}
 UE.Source.OneShot.OnFinished {type=Trigger}
]
pages [Default {
  input_defaults {
   MasterVolume=1
   Random.PitchShift.Base=0
   Random.PitchShift.Spread=0.5
   UE.Source.OnPlay=null
   Sounds [
    /MusterCoreWeapons/Items/Weapons/MusterPumpShotgun/Audio/Shotgun_MusterPump_Athena_OutOfAmmo_01
    /MusterCoreWeapons/Items/Weapons/MusterPumpShotgun/Audio/Shotgun_MusterPump_Athena_OutOfAmmo_02
    /MusterCoreWeapons/Items/Weapons/MusterPumpShotgun/Audio/Shotgun_MusterPump_Athena_OutOfAmmo_03
   ]
  }
  nodes [
   n1 = Array.Shuffle:WaveAsset:Array(In Array=^Sounds, Next=^UE.Source.OnPlay) {also {defaults {Enable Shared State="<MetaSoundBuilderResult.SUCCEEDED: 0>"}}}
   WavePlayer_0 = UE.Wave Player:Mono(
    Play=@n1.On Next
    Wave Asset=@n1.Value
    Pitch Shift=UE.RandomFloat(
     Max=UE.Add:Float(PrimaryOperand=^Random.PitchShift.Base, AdditionalOperands=^Random.PitchShift.Spread).Out
     Next=@n1.On Next
     Min=UE.Subtract:Float(PrimaryOperand=^Random.PitchShift.Base, AdditionalOperands=^Random.PitchShift.Spread).Out
    ).Value
   ) {also {version [1,1]}}
  ]
  links [{dst="@UE.Source.OneShot.OnFinished", src=WavePlayer_0.On Finished}, {dst="@UE.OutputFormat.Mono.Audio:0", src=WavePlayer_0.Out Mono}]
 }]
properties {
 assetImportData=&self:AssetImportData
 attenuationSettings=&AS_Fort_WeaponFoley
 maxDistance=2250
 sampleRate=48000
 soundClassObject=&Fort_WeaponFoley
 volume=0.55
}
